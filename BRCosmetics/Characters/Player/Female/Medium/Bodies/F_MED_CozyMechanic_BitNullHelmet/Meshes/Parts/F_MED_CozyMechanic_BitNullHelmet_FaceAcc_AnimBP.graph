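# Animation Blueprint for the face accessory part of the female CozyMechanic BitNullHelmet cosmetic, adding procedural and physics-driven motion.
# The pose copied from the attached parent runs through a Control Rig, then a RigidBody node simulating relative to spine_05 with gravity override enabled.
# Each update stores a Gravity Override vector from a nativized, parameter-asset-driven state function.
# The Control Rig, physics asset and parameters are reused from the base CozyMechanic_Bit body.
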AnimBlueprint /BRCosmetics/Characters/Player/Female/Medium/Bodies/F_MED_CozyMechanic_BitNullHelmet/Meshes/Parts/F_MED_CozyMechanic_BitNullHelmet_FaceAcc_AnimBP
target_skeleton=&/F_MED_CozyMechanic_BitNullHelmet_FaceAcc_Skeleton
graphs [
 AnimGraph {nodes [
   n1 = Control Rig(Source=Copy Pose From Mesh() {bUseAttachedParent=true}) {controlRigAssetReference {blueprintRigClass=/BRCosmetics/Characters/Player/Female/Medium/Bodies/F_MED_CozyMechanic_Bit/Meshes/Parts/F_MED_CozyMechanic_Bit_FaceAcc_CtrlRig.F_MED_CozyMechanic_Bit_FaceAcc_CtrlRig_C}}
   n2 = RigidBody!(ComponentPose=Local To Component(LocalPose=@n1), Alpha, OverrideWorldGravity:Vector="0,0,0", ExternalForce) -> (Pose) {
    bClampLinearTranslationLimitToRefPose=true
    bForceDisableCollisionBetweenConstraintBodies=true
    bOverrideWorldGravity=true
    overridePhysicsAsset=/BRCosmetics/Characters/Player/Female/Medium/Bodies/F_MED_CozyMechanic_Bit/Meshes/Parts/F_MED_CozyMechanic_Bit_FaceAcc_Physics
    simulationSpace=BaseBoneSpace
    alphaBoolBlend {blendOption=Linear}
    baseBoneRef {boneName=spine_05}
    simSpaceSettings {
     dampingAlpha=0
     maxAngularAcceleration=80
     maxAngularVelocity=8
     maxLinearAcceleration=700
     maxLinearVelocity=30
     velocityScaleZ=0.7
     worldAlpha=1
     externalLinearDragV {x=0.5}
    }
   }
   Output Pose(Result=Component To Local(ComponentPose=@n2)) {layerGroup=None}
  ]}
 EventGraph {nodes [
   Event BlueprintUpdateAnimation()
   > n1 = "Gravity Override State Function (Parameter Asset) (Nativized)"(
    CharacterPartAnimInstance=Self-Reference()
    GameTimeInSeconds=GetGameTimeInSeconds()
    Parameters=/BRCosmetics/Characters/Player/Female/Medium/Bodies/F_MED_CozyMechanic_Bit/Meshes/M_MED_CozyMechanic_Bit_Parameters
   )
   > Set Gravity Override(Gravity Override=@n1)
  ]}
]
variables [Gravity Override {type=Vector}]
